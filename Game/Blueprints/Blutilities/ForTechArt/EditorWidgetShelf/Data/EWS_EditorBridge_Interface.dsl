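; Blueprint interface for the Editor Widget Shelf, a tech-art editor utility, that bridges shelf widgets to the editor side.
; It declares a single function, deleteShelfButton, which takes a user widget reference for the button to remove.
; The function carries only a default false return, so implementers supply the actual deletion logic.
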(fn deleteShelfButton
  (params (button :type User%20Widget%20Object%20Reference))
  (return false))
